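# Physics asset for the small female Outlander 01 player body mesh, defining its simplified collision and jointed physics bodies.
# Single capsules cover the root, pelvis, spine, clavicles, limbs, hands, feet, balls and some twist bones, with no head or neck body.
# Constraints link each body to its parent bone in a tree from root, all with uniform 45-degree swing and twist limits.
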PhysicsAsset /Game/Characters/Player/Female/Small/Bodies/F_SML_Outlander_01/Mesh/F_SML_Outlander_01_Physics
bodies {
 ball_l [{shapeType=Capsule, length=4.5299, radius=3.3156, center=<2.5033 -1.2927 -1.2724>, rotation=<rot 0.6062 -90 14.9574>}]
 ball_r [{shapeType=Capsule, length=4.5299, radius=3.3156, center=<-2.5033 1.2928 1.2723>, rotation=<rot 0.6062 -90 14.9574>}]
 calf_l [{shapeType=Capsule, length=15.5397, radius=6.5195, center=<-12.435 -3.213 -1.6506>, rotation=<rot 0 93.4288 -89.2295>}]
 calf_r [{shapeType=Capsule, length=15.6616, radius=6.246, center=<12.4048 3.1459 0.9507>, rotation=<rot 0 95.5735 -83.7017>}]
 calf_twist_01_l [{shapeType=Capsule, length=4.9821, radius=3.135, center=<-2.5402 0.3949 -0.0467>, rotation=<rot 0 -24.1211 -89.5089>}]
 calf_twist_01_r [{shapeType=Capsule, length=4.7869, radius=3.9447, center=<2.7575 -0.865 0.5568>, rotation=<rot 0 168.9156 -59.3981>}]
 clavicle_l [{shapeType=Capsule, length=4.6341, radius=8.8893, center=<4.4887 1.8438 0.2973>, rotation=<rot 0 88.5392 -87.8532>}]
 clavicle_r [{shapeType=Capsule, length=6.2958, radius=8.8893, center=<-2.8452 -1.8857 -0.2356>, rotation=<rot 0 -91.4607 -92.1469>}]
 foot_l [{shapeType=Capsule, length=7.4411, radius=5.5888, center=<-6.6371 1.6027 -0.3517>, rotation=<rot 0 -119.5779 -90.587>}]
 foot_r [{shapeType=Capsule, length=7.5807, radius=5.5888, center=<6.5167 -1.5345 0.3503>, rotation=<rot 0 60.422 -89.413>}]
 hand_l [{shapeType=Capsule, length=5.8709, radius=4.7405, center=<5.3607 -0.8412 -0.383>, rotation=<rot 4.7151 -90 -36.7905>}]
 hand_r [{shapeType=Capsule, length=7.2444, radius=9.6916, center=<-3.5013 -3.6373 -0.635>, rotation=<rot 0 149.471 -87.4036>}]
 lowerarm_l [{shapeType=Capsule, length=8.8987, radius=5.1702, center=<10.0233 -0.9435 0.1354>, rotation=<rot 0 -86.2269 -80.5626>}]
 lowerarm_r [{shapeType=Capsule, length=8.3431, radius=3.8822, center=<-9.3781 0.9129 -0.1858>, rotation=<rot 0 87.065 -89.7521>}]
 lowerarm_twist_02_r [{shapeType=Capsule, length=10.3123, radius=6.8886, center=<-4.8258 0.5317 -2.8081>, rotation=<rot 0 -96.5238 -86.5921>}]
 pelvis [{shapeType=Capsule, length=17.5912, radius=11.7793, center=<0.3294 -0.4071 3.0598>, rotation=<rot 14.1934 -90 6.5854>}]
 root [{shapeType=Capsule, length=0.505, radius=0.505, rotation=<rot -0.5903 -90 0>}]
 spine_01 [{shapeType=Capsule, length=3.334, radius=9.7663, center=<4.0938 0.5964 0.0007>, rotation=<rot 0 97.1735 -90>}]
 spine_02 [{shapeType=Capsule, length=1.931, radius=8.6907, center=<0.5368 0.4129 -0.0099>, rotation=<rot 0 87.7267 -90>}]
 spine_03 [{shapeType=Capsule, length=4.1213, radius=7.825, center=<-0.7943 1.473 -0.0015>, rotation=<rot 0 89.713 -90>}]
 spine_04 [{shapeType=Capsule, length=4.4663, radius=10.1276, center=<0.2805 0.8007 0>, rotation=<rot 0 85.7608 -90>}]
 spine_05 [{shapeType=Capsule, length=0, radius=15.6466, center=<2.5589 0.8638 -2.1418>, rotation=<rot 0 -3.9249 -80.7422>}]
 thigh_l [{shapeType=Capsule, length=9.2973, radius=6.5986, center=<-32.277 -0.8 2.2171>, rotation=<rot 0 -91.3221 -78.2938>}]
 thigh_r [{shapeType=Capsule, length=9.2957, radius=6.2727, center=<32.2355 0.547 -2.2471>, rotation=<rot 0 -90.9794 -78.054>}]
 thigh_twist_01_l [{shapeType=Capsule, length=18.49, radius=11.6721, center=<1.1846 3.0281 -4.5678>, rotation=<rot 0 -57.9494 -63.5225>}]
 thigh_twist_01_r [{shapeType=Capsule, length=13.781, radius=9.6143, center=<1.2796 -2.2453 0.008>, rotation=<rot 0 -101.0699 -67.1129>}]
 upperarm_l [{shapeType=Capsule, length=7.9838, radius=3.8556, center=<18.1367 -1.131 0.2891>, rotation=<rot 0 -86.3848 -69.7526>}]
 upperarm_r [{shapeType=Capsule, length=6.3401, radius=3.815, center=<-19.2394 0.9771 0.3045>, rotation=<rot 0 -91.9956 -76.6214>}]
 upperarm_twist_01_l [{shapeType=Capsule, length=10.8838, radius=5.5215, center=<3.0487 0.3487 0.8347>, rotation=<rot 0 -83.0564 -72.7193>}]
 upperarm_twist_01_r [{shapeType=Capsule, length=10.1462, radius=4.6683, center=<-3.8986 0.317 0.5928>, rotation=<rot 0 -93.7256 -73.4206>}]
}
constraints [
 {bone1Name=pelvis, bone2Name=root, swing1LimitDegrees=45, swing2LimitDegrees=45, twistLimitDegrees=45}
 {bone1Name=spine_01, bone2Name=pelvis, swing1LimitDegrees=45, swing2LimitDegrees=45, twistLimitDegrees=45}
 {bone1Name=spine_02, bone2Name=spine_01, swing1LimitDegrees=45, swing2LimitDegrees=45, twistLimitDegrees=45}
 {bone1Name=spine_03, bone2Name=spine_02, swing1LimitDegrees=45, swing2LimitDegrees=45, twistLimitDegrees=45}
 {bone1Name=spine_04, bone2Name=spine_03, swing1LimitDegrees=45, swing2LimitDegrees=45, twistLimitDegrees=45}
 {bone1Name=spine_05, bone2Name=spine_04, swing1LimitDegrees=45, swing2LimitDegrees=45, twistLimitDegrees=45}
 {bone1Name=clavicle_l, bone2Name=spine_05, swing1LimitDegrees=45, swing2LimitDegrees=45, twistLimitDegrees=45}
 {bone1Name=upperarm_l, bone2Name=clavicle_l, swing1LimitDegrees=45, swing2LimitDegrees=45, twistLimitDegrees=45}
 {bone1Name=lowerarm_l, bone2Name=upperarm_l, swing1LimitDegrees=45, swing2LimitDegrees=45, twistLimitDegrees=45}
 {bone1Name=hand_l, bone2Name=lowerarm_l, swing1LimitDegrees=45, swing2LimitDegrees=45, twistLimitDegrees=45}
 {bone1Name=upperarm_twist_01_l, bone2Name=upperarm_l, swing1LimitDegrees=45, swing2LimitDegrees=45, twistLimitDegrees=45}
 {bone1Name=clavicle_r, bone2Name=spine_05, swing1LimitDegrees=45, swing2LimitDegrees=45, twistLimitDegrees=45}
 {bone1Name=upperarm_r, bone2Name=clavicle_r, swing1LimitDegrees=45, swing2LimitDegrees=45, twistLimitDegrees=45}
 {bone1Name=lowerarm_r, bone2Name=upperarm_r, swing1LimitDegrees=45, swing2LimitDegrees=45, twistLimitDegrees=45}
 {bone1Name=hand_r, bone2Name=lowerarm_r, swing1LimitDegrees=45, swing2LimitDegrees=45, twistLimitDegrees=45}
 {bone1Name=lowerarm_twist_02_r, bone2Name=lowerarm_r, swing1LimitDegrees=45, swing2LimitDegrees=45, twistLimitDegrees=45}
 {bone1Name=upperarm_twist_01_r, bone2Name=upperarm_r, swing1LimitDegrees=45, swing2LimitDegrees=45, twistLimitDegrees=45}
 {bone1Name=thigh_l, bone2Name=pelvis, swing1LimitDegrees=45, swing2LimitDegrees=45, twistLimitDegrees=45}
 {bone1Name=calf_l, bone2Name=thigh_l, swing1LimitDegrees=45, swing2LimitDegrees=45, twistLimitDegrees=45}
 {bone1Name=calf_twist_01_l, bone2Name=calf_l, swing1LimitDegrees=45, swing2LimitDegrees=45, twistLimitDegrees=45}
 {bone1Name=foot_l, bone2Name=calf_l, swing1LimitDegrees=45, swing2LimitDegrees=45, twistLimitDegrees=45}
 {bone1Name=ball_l, bone2Name=foot_l, swing1LimitDegrees=45, swing2LimitDegrees=45, twistLimitDegrees=45}
 {bone1Name=thigh_twist_01_l, bone2Name=thigh_l, swing1LimitDegrees=45, swing2LimitDegrees=45, twistLimitDegrees=45}
 {bone1Name=thigh_r, bone2Name=pelvis, swing1LimitDegrees=45, swing2LimitDegrees=45, twistLimitDegrees=45}
 {bone1Name=calf_r, bone2Name=thigh_r, swing1LimitDegrees=45, swing2LimitDegrees=45, twistLimitDegrees=45}
 {bone1Name=calf_twist_01_r, bone2Name=calf_r, swing1LimitDegrees=45, swing2LimitDegrees=45, twistLimitDegrees=45}
 {bone1Name=foot_r, bone2Name=calf_r, swing1LimitDegrees=45, swing2LimitDegrees=45, twistLimitDegrees=45}
 {bone1Name=ball_r, bone2Name=foot_r, swing1LimitDegrees=45, swing2LimitDegrees=45, twistLimitDegrees=45}
 {bone1Name=thigh_twist_01_r, bone2Name=thigh_r, swing1LimitDegrees=45, swing2LimitDegrees=45, twistLimitDegrees=45}
]
